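# Animation Blueprint for the GarageBand backpack cosmetic, adding secondary motion to its drumsticks, cymbal and kick pedal, including extra flutter while skydiving.
# Chained Anim Dynamics nodes drive those bones, then Copy Bone Delta passes doubled pedal-hinge rotation to the kick pivot.
# The update event turns pawn height into Perlin-noise Y force, damped under parachute and zero outside skydiving, scaled per node.
# Balloon Trail/Look At and a Copy Bone node are unconnected to the output.
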AnimBlueprint /BRCosmetics/Accessories/FORT_Backpacks/M_MED_GarageBand_Backpack/Meshes/M_MED_GarageBand_Backpack_Skeleton_AnimBP
target_skeleton=/Game/Accessories/FORT_Backpacks/M_MED_GarageBand_Backpack/Meshes/M_MED_GarageBand_Backpack_Skeleton
graphs [
 AnimGraph {nodes [
   Copy Bone - Source Bone: dyn_pedalHinge - Target Bone: dyn_kickPivot!(ComponentPose:Component Space Pose Link Structure) -> (Pose:Component Space Pose Link Structure) {
    is=title AnimGraphNode_CopyBone
    alpha=10
    bCopyRotation=true
    controlSpace=BCS_ParentBoneSpace
    alphaBoolBlend {blendOption=Linear}
    sourceBone {boneName=dyn_pedalHinge}
    targetBone {boneName=dyn_kickPivot}
   }
   n1 = Get falling_anim_dyn_up_force_vec3() -> (falling_anim_dyn_up_force_vec3:Vector)
   n2 = vector * float(A=@n1, B=0.15)
   n3 = Trail controller - Bone: dyn_balloon_base_02_01!(ComponentPose, Alpha=1, FakeVelocity:Vector=@n2) -> (Pose) {
    bUsePlanarLimit=true
    chainBoneAxis=Z
    relaxationSpeedScale=2
    baseJoint {boneName=dyn_balloon_base_02}
    planarLimits [
     {drivingBone {boneName=dyn_balloon_base_02}, planeTransform {location=<0 0 14>, rotation=<rot -33 -0.0001 -21.9999>, scale=<1 1 1>}}
     {drivingBone {boneName=backpack_root}, planeTransform {location=<20 -8 0>, rotation=<rot 0 5.0001 -89.9999>, scale=<1 1 1>}}
    ]
    rotationLimits [{limitMax=<180 180 180>}, {limitMax=<180 180 180>}]
    rotationOffsets [<0 0 0>, <0 0 0>]
    trailBone {boneName=dyn_balloon_base_02_01}
   }
   Look At - Bone: dyn_balloon_base_02_03_Aim(ComponentPose=@n3, Alpha=1) {
    alphaBoolBlend {blendOption=Linear}
    boneToModify {boneName=dyn_balloon_base_02_03_Aim}
    lookAtLocation {x=0}
    lookAtTarget {boneReference {boneName=dyn_balloon_base_02}}
    lookAt_Axis {axis {y=0, z=-1}}
    lookUp_Axis {axis=<0 1>}
   }
   n4 = Anim Dynamics - Bone: dyn_stickA(ComponentPose=Local To Component(), Alpha=0.5, ExternalForce=vector * float(A=@n1, B=0.05)) {
    angularDampingOverride=0.99
    angularSpringConstant=10
    bAngularSpring=true
    bGravityOverrideInSimSpace=true
    bOverrideAngularDamping=true
    alphaBoolBlend {blendOption=Linear}
    boundBone {boneName=dyn_stickA}
    componentAppliedLinearAccClamp=<20 20 20>
    componentLinearAccScale=<1 1 1>
    componentLinearVelScale=<1 1 1>
    physicsBodyDefinitions [{
      sphereCollisionRadius=0
      boundBone {boneName=dyn_stickA}
      boxExtents=<6 6 2>
      constraintSetup {angularLimitsMax=<10.8 3.6 0>, angularLimitsMin=<-43.2 -8 0>, angularTarget=<0 0 0>, linearAxesMax=<0.1 1 0>, linearAxesMin=<0 -1 0>}
      localJointOffset=<0 0 -3>
     }]
   }
   n5 = Anim Dynamics - Bone: dyn_stickB(ComponentPose=@n4, Alpha=0.5, ExternalForce=vector * float(A=@n1, B=-0.05)) {
    angularDampingOverride=0.99
    angularSpringConstant=10
    bAngularSpring=true
    bGravityOverrideInSimSpace=true
    bOverrideAngularDamping=true
    alphaBoolBlend {blendOption=Linear}
    boundBone {boneName=dyn_stickB}
    componentAppliedLinearAccClamp=<40 40 40>
    componentLinearAccScale=<1 1 1>
    componentLinearVelScale=<1 1 1>
    physicsBodyDefinitions [{
      sphereCollisionRadius=0
      boundBone {boneName=dyn_stickB}
      boxExtents=<6 6 2>
      constraintSetup {angularLimitsMax=<28.8 14.4 0>, angularLimitsMin=<-28.8 -4 0>, angularTarget=<0 0 0>, linearAxesMax=<0 1 0>, linearAxesMin=<-0.1 -1 0>}
      localJointOffset=<0 0 -3>
     }]
   }
   n6 = Anim Dynamics - Bone: dyn_cymbal(ComponentPose=@n5, Alpha=0.1, ExternalForce=vector * float(A=@n1, B=-0.75)) {
    bGravityOverrideInSimSpace=true
    alphaBoolBlend {blendOption=Linear}
    boundBone {boneName=dyn_cymbal}
    componentAppliedLinearAccClamp=<1000 1000 1000>
    componentLinearAccScale=<1 1 1>
    componentLinearVelScale=<1 1 1>
    physicsBodyDefinitions [{
      sphereCollisionRadius=0
      boundBone {boneName=dyn_cymbal}
      boxExtents=<8 8 2>
      constraintSetup {angularLimitsMax=<43.2 43.2 0>, angularLimitsMin=<-36 -32.4 0>, angularTarget=<0 0 0>}
      localJointOffset=<0 0 -3>
     }]
   }
   n7 = Anim Dynamics - Bone: dyn_pedalBase(ComponentPose=@n6, Alpha=0.5, ExternalForce=@n2) {
    angularSpringConstant=1000
    bAngularSpring=true
    bGravityOverrideInSimSpace=true
    bUsePlanarLimit=false
    alphaBoolBlend {blendOption=Linear}
    boundBone {boneName=dyn_pedalBase}
    componentAppliedLinearAccClamp=<1000 1000 1000>
    componentLinearAccScale=<1 1 1>
    componentLinearVelScale=<1 1 1>
    physicsBodyDefinitions [{
      sphereCollisionRadius=0
      boundBone {boneName=dyn_pedalBase}
      boxExtents=<16 16 4>
      constraintSetup {twistAxis=AxisY, angularLimitsMax=<3 0 0>, angularLimitsMin=<-3 0 0>, angularTarget=<0 0 0>}
      localJointOffset=<0 0 -3>
     }]
   }
   n8 = Anim Dynamics - Bone: dyn_pedalHinge(ComponentPose=@n7, Alpha=0.1, ExternalForce=vector * float(A=@n1, B=0.75)) {
    angularSpringConstant=10
    bAngularSpring=true
    bGravityOverrideInSimSpace=true
    bUsePlanarLimit=false
    simulationSpace=BoneRelative
    alphaBoolBlend {blendOption=Linear}
    boundBone {boneName=dyn_pedalHinge}
    componentAppliedLinearAccClamp=<1000 1000 1000>
    componentLinearAccScale=<1 1 1>
    componentLinearVelScale=<1 1 1>
    physicsBodyDefinitions [{
      sphereCollisionRadius=0
      boundBone {boneName=dyn_pedalHinge}
      boxExtents=<18 18 8>
      constraintSetup {twistAxis=AxisY, angularLimitsMax=<26.6 2 0>, angularLimitsMin=<-22 -2 0>, angularTarget=<0 0 0>}
      localJointOffset=<0 1 -3>
     }]
    relativeSpaceBone {boneName=dyn_pedalBase}
   }
   n9 = Copy Bone Delta - Source Bone: dyn_pedalHinge - Target Bone: dyn_kickPivot!(ComponentPose:Component Space Pose Link Structure=@n8, Alpha:float=1) -> (Pose:Component Space Pose Link Structure) {
    is=title AnimGraphNode_CopyBoneDelta
    bCopyRotation=true
    copyMode=Copy
    rotationMultiplier=2
    alphaBoolBlend {blendOption=Linear}
    sourceBone {boneName=dyn_pedalHinge}
    targetBone {boneName=dyn_kickPivot}
   }
   Output Pose(Result=Component To Local(ComponentPose=@n9)) {layerGroup=None}
  ]}
 EventGraph {nodes [
   Event BlueprintUpdateAnimation()
   n1 = Get PawnLastLocation() -> (PawnLastLocation:Vector)
   n2 = SelectFloat(
    A=Make Perlin Noise and Remap!(Value:float=BreakVector(InVec=@n1).Z, RangeOutMin:float=-25000, RangeOutMax:float=25000) -> (ReturnValue:float)
    B=0
    bPickA=Get bIsSkyDiving()
   )
   n3 = SelectFloat(A=float * float(A=@n2, B=0.1), B=@n2, bPickA=Get bIsParachuteOpened())
   > Set falling_anim_dyn_up_force_vec3(falling_anim_dyn_up_force_vec3:Vector=vector * float(A=MakeVector(Y=@n3), B=0.75)) -> (Output_Get:Vector)
  ]}
]
variables [falling_anim_dyn_up_force_vec3 {type=Vector}]
